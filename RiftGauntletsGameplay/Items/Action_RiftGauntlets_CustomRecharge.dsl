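; Custom recharge rule for the Rift Gauntlets item, deciding whether the weapon regains ammo for a given player controller.
; Its ShouldWeaponRechargeAmmo function checks the player pawn for the on-zipline gameplay tag from the zipline deployer.
; Recharge is blocked while that tag is present and allowed otherwise.
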(fn ShouldWeaponRechargeAmmo
  (params (PlayerController :type Fort%20Player%20Controller%20Object%20Reference))
  (Utilities|FlowControl|Sequence
    (:then_0
      (if (GameplayTags|HasMatchingGameplayTag (Game|Player|GetPlayerPawn PlayerController) "(TagName=\"Granted.Athena.Item.ZiplineDeployer.OnZipline\")")
        (return false)))
    (:then_1
      (return true))))
