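; Rift Gauntlets weapon Blueprint that manages out-of-ammo cooldown visuals and the weapon's break-and-remove sequence.
; On equip, the server watches local ammo; at zero it replicates the cooldown's end server time, which clients convert to a material "Duration" on the pawn's Gameplay mesh.
; DestroyWeapon replicates a flag that fires the break gameplay cue, then removes the item after a short timer.
; EndPlay forces that removal if the timer is still pending.
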(event FortWeapon|EventOnEquippedAndAttached
  ;[0.009399 0.15 0.02427 0.5] OnEquippedAndAttached: Bind to know when the local ammo count has changed.
  (|CallParentFunction)
  (Networking|SwitchHasAuthority
    (:Authority
      (EventDispatchers|BindEventtoOnLocalAmmoChanged (EventDispatchers|CreateEvent :!event Custom|HandleLocalAmmoChanged))
      (CallFunction|HandleLocalAmmoChanged (FortWeapon|GetLocalMagazineAmmoCount)))))

(event Custom|HandleLocalAmmoChanged
  (params (LocalCount :type Integer) (LocalRemaining :type Integer))
  (bind _returnvalue (FortWeapon|GetInventoryItem))
  ;[0.04376 0.1453 0.15 0.5] HandleLocalAmmoChanged: When the local ammo hits zero, update the Out Of Ammo End Server Time so the clients can update the visuals accordingly.
  (Networking|SwitchHasAuthority
    (:Authority
      (if (== LocalCount 0)
        (Utilities|IsValid _returnvalue
          (:Is%20Valid
            (bind (_cooldownduration _returnvalue_2) (Item|GetCoolDownTimeRemaining _returnvalue (Utilities|Casting|CastToFortPlayerController (Game|GetInstigatorController))))
            (Variables|Transient|SetOutOfAmmoEndServerTime (+ (Fortnite|Sync|Time|GetServerWorldTimeSeconds) _returnvalue_2))))))))

(event FortWeapon|EventOnUnEquippedAndDetached
  ;[0.15 0.02159 0.1342 0.5] OnUnequippedAndDetached: Clean up.
  (|CallParentFunction)
  (Networking|SwitchHasAuthority
    (:Authority
      (EventDispatchers|UnbindEventfromOnLocalAmmoChanged (EventDispatchers|CreateEvent :!event Custom|HandleLocalAmmoChanged)))))

(event Custom|DestroyWeapon
  ;[0.02663 0.08077 0.15 0.5] DestroyWeapon: Remove the weapon from the player's inventory after a short duration.
  (Networking|SwitchHasAuthority
    (:Authority
      (if (not (Utilities|Time|IsTimerActivebyHandle (Variables|Transient|GetDelayBeforeWeaponDestroyTimerHandle)))
        (bind _returnvalue (Utilities|Time|SetTimerbyEvent (EventDispatchers|CreateEvent :!event Custom|RemoveItemFromOwnerPlayerInventory) 0.4))
        (Variables|Transient|SetDelayBeforeWeaponDestroyTimerHandle _returnvalue)
        (Variables|Transient|SetbDestroyWeaponCosmetic true)
        (Networking|FlushNetDormancy)))))

(event Custom|RemoveItemFromOwnerPlayerInventory
  ;[0.06453 0.15 0.04561 0.5] RemoveItemFromOwnerPlayerInventory: Remove the item from the owning player's inventory.
  (Networking|SwitchHasAuthority
    (:Authority
      (Inventory|RemoveItemFromPlayerByGuid (Game|GetInstigatorController) (FortWeapon|GetInventoryGUID) -1 true))))

(event EventEndPlay
  (params (EndPlayReason :type EEndPlayReason%20Enum :default "Destroyed"))
  ;[0.15 0.08448 0.01056 0.5] EndPlay: If for some reason the weapon is being destroyed and the destroy timer is still active, remove the item from the owner.
  (Utilities|FlowControl|Sequence
    (:then_0
      (Networking|SwitchHasAuthority
        (:Authority
          (if (Utilities|Time|IsTimerActivebyHandle (Variables|Transient|GetDelayBeforeWeaponDestroyTimerHandle))
            (CallFunction|RemoveItemfromOwnerPlayerInventory)))))
    (:then_1
      (|CallParentFunction EndPlayReason))))

(event Custom|UpdateOutOfAmmoMaterialVisualBasedOnCooldown
  (params (CooldownTimeRemaining :type Float%20%28double-precision%29))
  ;[0.1133 0.05641 0.15 0.5] UpdateOutOfAmmoMaterialVisualBasedOnCooldown: Update the material visual.
  (Networking|CanExecuteCosmeticEvents
    (:True
      (bind _asfort_player_pawn (Utilities|Casting|CastToFortPlayerPawn (Game|GetInstigator)))
      (Utilities|IsValid (Customization|GetSkeletalMeshforPartType _asfort_player_pawn "Gameplay" :!class /Script/FortniteGame.FortPlayerPawn)
        (:Is%20Valid
          (Rendering|Material|SetScalarParameterforCustomPrimitiveData (Customization|GetSkeletalMeshforPartType _asfort_player_pawn "Gameplay" :!class /Script/FortniteGame.FortPlayerPawn) "Duration" CooldownTimeRemaining))))))

(fn OnRep_bDestroyWeaponCosmetic
  ;[0.03218 0.03505 0.15 0.5] Run the gameplay cue that shows the item breaking.
  (if (Variables|Transient|GetbDestroyWeaponCosmetic)
    (GameplayCue|ExecuteGameplayCueOnActor(Burst) self (Variables|GameplayCue|GetWeaponDestroyedGameplayCue) (Ability|GameplayCue|MakeGameplayCueParameters :NormalizedMagnitude 0.0 :RawMagnitude 0.0 :Location (Transformation|GetActorLocation) :Normal "0,0,0" :Instigator (Game|GetInstigator) :SourceObject self :GameplayEffectLevel 1 :AbilityLevel 1 :bReplicateLocationWhenUsingMinimalRepProxy false))))

(fn OnRep_OutOfAmmoEndServerTime
  (bind _returnvalue (Fortnite|Sync|Time|GetServerWorldTimeSeconds))
  (bind _outofammoendservertime (Variables|Transient|GetOutOfAmmoEndServerTime))
  (bind _returnvalue_2 (- _outofammoendservertime _returnvalue))
  (Networking|CanExecuteCosmeticEvents
    (:True
      (if (> _returnvalue_2 0)
        (CallFunction|UpdateOutOfAmmoMaterialVisualBasedonCooldown _returnvalue_2)))))

(fn ConstructionScript)
